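# Post-process style Anim Blueprint for the CleverEdge Purple medium male body, copying the parent mesh pose and layering cosmetic fixes on top.
# The pose runs through a Control Rig, a backpack bone offset, a cape bone offset and a RigidBody node using cape colliders.
# The cape offset eases with crouching, despite its isWearingCape_0 alpha name.
# On customization load, characters with tails hide the tail mesh for capes or long backblings.
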AnimBlueprint /BRCosmetics/Characters/Player/Male/Medium/Bodies/M_MED_CleverEdge_Purple/Meshes/M_MED_CleverEdge_Purple_AnimBP
target_skeleton=&/M_MED_CleverEdge_Purple_Skeleton
graphs [
 AnimGraph {nodes [
   n1 = Control Rig(Source=Copy Pose From Mesh() {bUseAttachedParent=true}) {
    alphaBoolBlend {blendOption=Linear}
    controlRigAssetReference {blueprintRigClass=&/M_MED_CleverEdge_Purple_ControlRig.M_MED_CleverEdge_Purple_ControlRig_C}
   }
   n2 = "Transform (Modify) Bone - Bone: attach_backpack"(ComponentPose=Local To Component(LocalPose=@n1), Alpha=1, Translation="0,-1,0", Rotation="0,0,0", Scale="1,1,1") {
    translationMode=BMM_Additive
    translationSpace=BCS_BoneSpace
    alphaBoolBlend {blendOption=Linear}
    boneToModify {boneName=attach_backpack}
    translation {y=-1}
   }
   n3 = "Transform (Modify) Bone - Bone: attach_cape"(ComponentPose=@n2, Alpha=Get isWearingCape_0() -> (isWearingCape_0:double), Translation="0,-2,-4", Rotation="0,0,15", Scale="1,1,1") {
    rotationMode=BMM_Additive
    rotationSpace=BCS_BoneSpace
    translationMode=BMM_Additive
    translationSpace=BCS_BoneSpace
    alphaBoolBlend {blendOption=Linear}
    alphaScaleBiasClamp {bInterpResult=true, interpSpeedDecreasing=15, interpSpeedIncreasing=15}
    boneToModify {boneName=attach_cape}
    rotation {roll=15}
    translation {y=-2, z=-4}
   }
   n4 = RigidBody(ComponentPose=@n3) {
    overridePhysicsAsset=&/M_MED_CleverEdge_Purple_Capecolliders
    alphaBoolBlend {blendOption=Linear}
   }
   Output Pose(Result=Component To Local(ComponentPose=@n4)) {layerGroup=None}
  ]}
 EventGraph {nodes [
   Event BlueprintUpdateAnimation()
   > Set isWearingCape_0(isWearingCape_0:double="To Float (Boolean)"(InBool=Get bIsCrouching())) -> (Output_Get:double)
   Event OnCharacterCustomizationLoaded()
   > Set BodyMesh(BodyMesh=GetOwnerSkeletalMeshComponent())
   > n1 = Cast To PlayerPawn_Athena(Object=TryGetPawnOwner())
   > n2 = Set PawnOwner(PawnOwner:Player Pawn Athena Object Reference=@n1) -> (Output_Get:Player Pawn Athena Object Reference)
   > Set TailMesh(TailMesh:Skeletal Mesh Component Object Reference=GetSkeletalMeshForPartType(self=@n2, PartType=MiscOrTail)) -> (Output_Get:Skeletal Mesh Component Object Reference)
   > Set characterHasTail(characterHasTail:bool=CheckForTail!(self:Custom Character Part Anim Instance Object Reference) -> (bHasTail:bool)) -> (Output_Get:bool)
   n3 = CheckForBackbling()
   > Set isWearingBackbling(isWearingBackbling=@n3)
   > Set isWearingCape(isWearingCape=@n3.bIsBackblingACape)
   > Set isBackblingLong(isBackblingLong:bool=@n3.bIsBackblingLong) -> (Output_Get:bool)
   > Branch(Condition=Get characterHasTail() -> (characterHasTail:bool))
   n4 = Get isWearingBackbling()
   n5 = Get isBackblingLong() -> (isBackblingLong:bool)
   > n6 = Branch(Condition=OR Boolean(A=Get isWearingCape() -> (isWearingCape:bool), B=AND Boolean(A=@n4, B=@n5)))
   n7 = Get TailMesh() -> (TailMesh:Skeletal Mesh Component Object Reference)
   > SetVisibility!(execute, self:Scene Component Object Reference=@n7, bNewVisibility:bool=false, bPropagateToChildren:bool=false) -> (then)
   > Set cape and tail combo(cape and tail combo:bool=false) -> (Output_Get:bool)
   SetVisibility(execute=@n6.else, self=@n7, bNewVisibility=true, bPropagateToChildren=false)
   > Set cape and tail combo(cape and tail combo=true)
   Get isWearingCape()
  ]}
]
variables [
 BodyMesh {type=SkeletalMeshComponent}
 cape and tail combo {type=boolean}
 characterHasTail {type=boolean}
 isBackblingLong {type=boolean}
 isWearingBackbling {type=boolean}
 isWearingCape {type=boolean}
 isWearingCape_0 {type=number}
 PawnOwner {type=PlayerPawn_Athena_C}
 TailMesh {type=SkeletalMeshComponent}
]
